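# Material function in the lava lamp prop folder that draws one soft, animated drop as an inverted sphere mask.
# It builds screen UVs centred on the object's clip-space position and scales them by view depth.
# A time-driven sine moves the drop vertically, while its horizontal position blends toward a mirrored value based on object orientation relative to the camera.
# Inputs set position, timing, travel and radius, with travel and radius divided by FOV.
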MaterialFunction /Game/Environments/Helios/Props/GreenTown/GreenTown_LavaLamp_A/Materials/MF_ScreenUVDrops
graph {nodes [
  Constant()
  n1 = ConstantBiasScale(None=Add(A=Constant2Vector(), B=ScreenPosition().ViewportUV)) {bias=-0.5, scale=2}
  n2 = MaterialFunctionCall(Float2=ViewSize()) {materialFunction=&BreakOutFloat2Components}
  n3 = AppendVector(A=Constant() {r=1}, B=Divide(A=@n2.G, B=@n2.G))
  n4 = Constant2Vector() {g=-0.5, r=0.5}
  n5 = ObjectPositionWS()
  n6 = MaterialFunctionCall(World Position=@n5) {materialFunction=&TransformToClipSpace}
  n7 = Subtract(A=ConstantBiasScale(None=Multiply(A=@n1, B=@n3)), B=Multiply(A=@n3, B=Multiply(A=@n4, B=@n6.Clip Space XY) {constA=1}))
  n8 = Divide(
   A=ScalarParameter() {defaultValue=6, parameterName=Screen Space Scale}
   B=ComponentMask(None=TransformPosition(None=@n5) {transformSourceType=TRANSFORMPOSSOURCE_World, transformType=TRANSFORMPOSSOURCE_View}) {b=true}
  ) {constA=100}
  n9 = Constant() {r=0.5}
  n10 = FunctionInput() {bUsePreviewValueAsDefault=true, inputName=DropPositionXY, inputType=FunctionInput_Scalar, previewValue=<0.2 0 0 1>}
  n11 = Subtract(A=@n9, B=Subtract(A=@n10, B=@n9) {constB=0.5})
  n12 = Constant3Vector() {constant=<rgba 0 1 0 0>}
  n13 = ConstantBiasScale(None=ComponentMask(None=Transform(None=@n12) {transformSourceType=TRANSFORMSOURCE_Local, transformType=TRANSFORM_Camera}) {r=true})
  n14 = Time()
  n15 = FunctionInput() {bUsePreviewValueAsDefault=true, inputName=DropTimeOffset, inputType=FunctionInput_Scalar, previewValue=<0.2 0 0 1>}
  n16 = FunctionInput() {bUsePreviewValueAsDefault=true, inputName=DropSpeedZ, inputType=FunctionInput_Scalar, previewValue=<0.06 0 0 1>}
  n17 = Multiply(A=Multiply(A=Add(A=@n14, B=@n15)) {constB=-1}, B=@n16) {constB=0.1}
  n18 = FunctionInput() {bUsePreviewValueAsDefault=true, inputName=DropInterval, inputType=FunctionInput_Scalar, previewValue=<1.2 0 0 1>}
  n19 = MaterialFunctionCall() {materialFunction=/Engine/Functions/Engine_MaterialFunctions02/Utility/FOV}
  n20 = Add(A=Multiply(A=Sine(None=@n17), B=Divide(A=@n18, B=@n19."FOV ")) {constB=2}) {constB=1.35}
  n21 = FunctionInput() {bUsePreviewValueAsDefault=true, inputName=DropHeightOffset, inputType=FunctionInput_Scalar}
  n22 = FunctionInput() {bUsePreviewValueAsDefault=true, inputName=DropRadius, inputType=FunctionInput_Scalar, previewValue=<2 0 0 1>}
  n23 = SphereMask(
   A=Multiply(
    A=MaterialFunctionCall(UVs=@n7, Texture Scale=@n8) {materialFunction=&ScaleUVsByCenter}.UVs
    B=Constant2Vector() {g=1, r=1}
   )
   B=AppendVector(A=LinearInterpolate(A=@n11, B=@n10, Alpha=@n13), B=Add(A=@n20, B=@n21))
   Radius=Divide(A=@n22, B=MaterialFunctionCall() {materialFunction=&FOV}."FOV ")
  ) {attenuationRadius=0.3, hardnessPercent=0}
  FunctionOutput(None=OneMinus(None=Power(Base=@n23) {constExponent=1}))
 ]}
