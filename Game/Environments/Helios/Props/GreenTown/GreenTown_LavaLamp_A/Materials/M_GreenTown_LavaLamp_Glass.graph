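# Lit material for the GreenTown lava lamp glass that fakes moving blobs inside it.
# Animated MF_ScreenUVDrops layers multiply into a metaball-like mask, with bottom padding built from a virtual-plane sphere mask and panning cloud noise.
# The mask drives a curve-atlas color and height-derived normal highlights, combined with glass Fresnel and ceiling gradient reroutes and compensated for eye adaptation.
# Panning, bump-offset bubble normals pulse in flatness over time.
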Material /Game/Environments/Helios/Props/GreenTown/GreenTown_LavaLamp_A/Materials/M_GreenTown_LavaLamp_Glass
graph {
 properties {bAutomaticallySetUsageInEditor=false, usedShadingModels=Default Lit}
 nodes [
  n1 = Constant3Vector() {constant=<rgba 0 0 1 0>}
  n2 = Constant3Vector() {constant=<rgba 1 0 0 0>}
  n3 = MaterialFunctionCall(
   Plane Normal Axis=Normalize(VectorInput=Transform(None=@n1) {transformSourceType=TRANSFORMSOURCE_Local})
   Plane X Axis=Normalize(VectorInput=Transform(None=@n2) {transformSourceType=TRANSFORMSOURCE_Local})
   Plane Center=TransformPosition(None=Constant3Vector() {constant=<rgba 0 0 3 0>}) {transformType=TRANSFORMPOSSOURCE_World}
   UV Scale=ScalarParameter() {defaultValue=10, parameterName=VirtualPlaneUVScale}
  ) {materialFunction=&VirtualPlaneCoordinates}
  n4 = Subtract(A=ComponentMask(None=@n3.UVs) {r=true}) {constB=0.5}
  n5 = Subtract(A=ComponentMask(None=@n3.UVs) {g=true}) {constB=0.5}
  n6 = Max(A=Abs(None=Sign(None=@n4)), B=Abs(None=Sign(None=@n5)))
  n7 = SphereMask(A=@n3.UVs, B=Constant2Vector() {g=0.5, r=0.5}, Radius=ScalarParameter() {defaultValue=3, parameterName=BottomPaddingRadius}) {attenuationRadius=2, hardnessPercent=0}
  n8 = Power(Base=OneMinus(None=Multiply(A=@n6, B=@n7))) {constExponent=1}
  n9 = Power(Base=@n8, Exp=ScalarParameter() {defaultValue=0.5, parameterName=BottomPaddingExp}) {constExponent=1}
  n10 = ScalarParameter() {parameterName=BottomPaddingNoise}
  n11 = Panner() {speedY=0.106}
  n12 = Constant2Vector() {g=0.5, r=1.0063}
  n13 = ConstantBiasScale(None=TextureSample(UVs=Multiply(A=@n11, B=@n12) {constB=2}) {samplerType=SAMPLERTYPE_LinearColor, texture=/Game/Packages/DS_Fortnite_Terrain_NoLOD/Textures/Cloud_Height5}.RGB) {bias=-0.5, scale=2}
  Add(A=@n9, B=Multiply(A=Step(Y=@n9) {constX=0.65}, B=Multiply(A=@n10, B=@n13)))
  Constant()
  MaterialFunctionCall() {materialFunction=/Engine/Functions/Engine_MaterialFunctions02/Utility/FOV}
  MaterialFunctionCall() {materialFunction=&FOV}
  n14 = Multiply(
   A=MaterialFunctionCall!(DropPositionXY, DropSpeedZ, DropInterval, DropTimeOffset, DropHeightOffset, DropRadius=Constant() {r=0.25}) {materialFunction=&/MF_ScreenUVDrops}.Result
   B=MaterialFunctionCall(
    DropPositionXY=ScalarParameter() {defaultValue=0.45, parameterName=Drop2_PositionXY}
    DropSpeedZ=ScalarParameter() {defaultValue=0.1, parameterName=Drop2_SpeedZ}
    DropInterval=ScalarParameter() {defaultValue=1.3, parameterName=Drop2_Interval}
    DropTimeOffset=ScalarParameter() {defaultValue=5, parameterName=Drop2_TimeOffset}
    DropRadius=ScalarParameter() {defaultValue=0.6, parameterName=Drop2_Radius}
   ) {materialFunction=&MF_ScreenUVDrops}.Result
  )
  n15 = MaterialFunctionCall(
   DropPositionXY=ScalarParameter() {defaultValue=0.7, parameterName=Drop5_PositionXY}
   DropSpeedZ=ScalarParameter() {defaultValue=0.04, parameterName=Drop5_SpeedZ}
   DropInterval=ScalarParameter() {defaultValue=0.9, parameterName=Drop5_Interval}
   DropTimeOffset=ScalarParameter() {defaultValue=0.5, parameterName=Drop5_TimeOffset}
   DropRadius=ScalarParameter() {defaultValue=0.3, parameterName=Drop5_Radius}
  ) {materialFunction=&MF_ScreenUVDrops}
  n16 = ScalarParameter() {defaultValue=0.15, parameterName=Drop6_PositionXY}
  n17 = ScalarParameter() {defaultValue=0.3, parameterName=Drop6_SpeedZ}
  n18 = ScalarParameter() {defaultValue=0.1, parameterName=Drop6_Interval}
  n19 = ScalarParameter() {defaultValue=3.5, parameterName=Drop6_TimeOffset}
  n20 = Constant() {r=-1}
  n21 = ScalarParameter() {defaultValue=0.4, parameterName=Drop6_Radius}
  n22 = Multiply(
   A=Multiply(
    A=Multiply(A=@n14, B=@n15.Result)
    B=MaterialFunctionCall(DropPositionXY=@n16, DropSpeedZ=@n17, DropInterval=@n18, DropTimeOffset=@n19, DropHeightOffset=@n20, DropRadius=@n21) {materialFunction=&MF_ScreenUVDrops}.Result
   )
   B=NamedRerouteUsage()
  )
  n23 = ScalarParameter() {defaultValue=-0.2, parameterName=Param_1}
  n24 = Saturate(None=Multiply(A=Add(A=@n22, B=@n23), B=ScalarParameter() {defaultValue=2, parameterName=Param_2}))
  n25 = ComponentMask(None=OneMinus(None=@n24)) {r=true}
  n26 = Transform(None=MaterialFunctionCall!(Height=@n25, Absolute World Position, World Space Vertex Normals) {materialFunction=/Engine/Functions/Engine_MaterialFunctions03/Procedurals/HeightToNormalSmooth}.WorldSpaceNormal) {transformSourceType=TRANSFORMSOURCE_World, transformType=TRANSFORM_Tangent}
  Multiply(A=@n26)
  Multiply(
   A=MaterialFunctionCall(
    DropPositionXY=ScalarParameter() {defaultValue=0.5, parameterName=Drop3_PositionXY}
    DropSpeedZ=ScalarParameter() {parameterName=Drop3_SpeedZ}
    DropInterval=ScalarParameter() {defaultValue=1, parameterName=Drop3_Interval}
    DropTimeOffset=ScalarParameter() {defaultValue=1, parameterName=Drop3_TimeOffset}
    DropHeightOffset=Constant() {r=2}
    DropRadius=ScalarParameter() {defaultValue=3, parameterName=Drop3_Radius}
   ) {materialFunction=&MF_ScreenUVDrops}.Result
   B=@n8
  )
  NamedRerouteDeclaration(None=Saturate(None=@n9)) {name=BottomPadding, nodeColor=<rgba 1 0.2812 0 1>}
  n27 = MaterialFunctionCall(Float3=ObjectBounds()) {materialFunction=&BreakOutFloat3Components}
  NamedRerouteDeclaration(None=AppendVector(A=Multiply(A=@n27.R) {constB=2}, B=@n27.B)) {name=ScreenSpaceBound, nodeColor=<rgba 0 1 0.25 1>}
  n28 = VectorParameter() {parameterName=HighlightColor, defaultValue=<rgba 1 0 5 1>}
  n29 = MaterialFunctionCall(Normal=@n26, Flatness=Constant() {r=0.2}) {materialFunction=&FlattenNormal}
  n30 = Multiply(A=Saturate(None=ComponentMask(None=@n29.Result) {g=true}), B=@n24)
  n31 = CurveAtlasRowParameter(CurveTime=@n25) {
   atlas=&/C_GreenTown_LavaLamp
   curve=&/C_GreenTown_LavaLampColor
   parameterName=ColorCurve
  }
  n32 = Multiply(A=VectorParameter() {parameterName=Param_2, defaultValue=<rgba 0.8333 0 0.117 1>}.RGB, B=NamedRerouteUsage())
  n33 = Add(A=NamedRerouteUsage(), B=Add(A=Multiply(A=@n28.RGB, B=@n30), B=Add(A=@n31, B=@n32)))
  n34 = ScalarParameter() {defaultValue=1, parameterName=EmissiveFinal}
  NamedRerouteDeclaration(None=EyeAdaptationInverse(LightValueInput=Multiply(A=@n33, B=@n34), AlphaInput=ScalarParameter() {defaultValue=0.65, parameterName=EyeAdaptionInverse}).EyeAdaptationInverse) {name=OUT, nodeColor=<rgba 0 1 0.9531 1>}
  n35 = Fresnel(ExponentIn=ScalarParameter() {defaultValue=1.5, parameterName=GlassFresnel_Exp}) {exponent=1.5}
  NamedRerouteDeclaration(None=LinearInterpolate(
    A=VectorParameter() {parameterName=GlassFresnel_OuterColor, defaultValue=<rgba 0.1667 0.2491 2 1>}.RGB
    B=VectorParameter() {parameterName=GlassFresnel_InnerColor, defaultValue=<rgba 0.0488 0.011 0.0781 1>}.RGB
    Alpha=OneMinus(None=@n35)
   )) {name=BG, nodeColor=<rgba 1 0.8672 0 1>}
  n36 = MaterialFunctionCall(Float2=TextureCoordinate()) {materialFunction=&BreakOutFloat2Components}
  NamedRerouteDeclaration(None=Add(A=SmoothStep(Value=@n36.G) {constMax=1.5, constMin=0.6}, B=SmoothStep(Value=@n36.G) {constMax=-0.5, constMin=0.3})) {name=Ceil, nodeColor=<rgba 0.3594 0 1 1>}
  n37 = TextureCoordinate()
  n38 = Multiply(A=ScalarParameter() {defaultValue=0.4, parameterName=BubbleTiling}, B=Constant2Vector() {g=1, r=2.5})
  n39 = BumpOffset(
   Coordinate=Panner(Coordinate=Multiply(A=@n37, B=@n38)) {speedY=0.4}
   Height=ScalarParameter() {defaultValue=2, parameterName=BubbleBumpOffsetHeight}
  )
  n40 = ScalarParameter() {defaultValue=0.7, parameterName=BubbleFlatteness}
  n41 = Multiply(A=MaterialFunctionCall(Speed=Constant() {r=0.5}) {materialFunction=&DebugTimeSine}.Result) {constB=0.3}
  NamedRerouteDeclaration(None=MaterialFunctionCall(
    Normal=TextureSample(UVs=@n39) {samplerType=SAMPLERTYPE_Normal, texture=/Game/Effects/Niagara/MasterMaterials/Liquid/T_Liquid_Bubbles_Normal}.RGB
    Flatness=Add(A=@n40, B=@n41)
   ) {materialFunction=&FlattenNormal}.Result) {name=Normal, nodeColor=<rgba 0 0.5312 1 1>}
  NamedRerouteUsage()
  NamedRerouteUsage()
  ScalarParameter() {defaultValue=1, parameterName=Specular}
  ScalarParameter() {defaultValue=0.15, parameterName=Roughness}
  ScalarParameter() {parameterName=Metalness}
 ]
}
